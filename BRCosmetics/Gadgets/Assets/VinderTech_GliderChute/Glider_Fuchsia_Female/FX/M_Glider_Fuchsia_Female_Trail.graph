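# Translucent, unlit, two-sided trail material for the Fuchsia Female glider, built for Niagara mesh particles.
# Emissive combines a curve-atlas gradient, scrolling smoke noise and particle color, with a simple-glow blend and time-of-day tint.
# Dynamic parameters shape strip width and drive an appear wipe on opacity, softened by depth fade.
# World position offset layers masked panning noise with optional SplineThicken, plus a switchable bend from Niagara parameters.
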Material /BRCosmetics/Gadgets/Assets/VinderTech_GliderChute/Glider_Fuchsia_Female/FX/M_Glider_Fuchsia_Female_Trail
graph {
 properties {
  bUsedWithNiagaraMeshParticles=true
  blendMode=BLEND_Translucent
  pixelDepthOffsetMode=PDOM_Legacy
  shadingModel=MSM_Unlit
  twoSided=true
  usedShadingModels=Unlit
 }
 nodes [
  n1 = VectorParameter() {parameterName="Color_Morning (V3)", defaultValue=<rgba 0.224 0.224 0.224 0>}
  n2 = Constant() {r=1}
  n3 = VectorParameter() {parameterName="Color_Evening (V3)", defaultValue=<rgba 0.3 0.3 0.3 0>}
  n4 = VectorParameter() {parameterName="Color_Night (V3)", defaultValue=<rgba 0.45 0.2461 0.3691 0>}
  n5 = TextureCoordinate() {uTiling=0.5, vTiling=0.5}
  n6 = CurveAtlasRowParameter(CurveTime=ComponentMask(None=TextureCoordinate()) {g=true}) {
   atlas=/Game/Athena/LinearColorCurves/A_Atlas_Turtleneck_Glider
   curve=/Game/Gadgets/Assets/VinderTech_GliderChute/Glider_Fuchsia_Female/FX/Glider_Fuchsia_Female_Trail_Curve
   defaultValue=18
  }
  n7 = ScalarParameter() {defaultValue=1.1, parameterName=Color Intensity}
  n8 = Desaturation(None=Add(B=Multiply(A=@n6, B=@n7) {constB=1.1}), Fraction=ScalarParameter() {parameterName=Des})
  n9 = Multiply(
   A=LinearInterpolate(Alpha=TextureSample(UVs=@n5) {samplerType=SAMPLERTYPE_Grayscale, texture=&LowResBlurredNoise}.R) {constB=3}
   B=@n8
  )
  n10 = TextureCoordinate()
  n11 = DynamicParameter() {defaultValue=<rgba 1 1 0.4 1>, paramNames [Spline Ticken Intensity, Width, Width bottom, Appear]}
  n12 = LinearInterpolate(
   A=ScalarParameter() {defaultValue=0.1, parameterName=ss}
   B=@n11.Width bottom
   Alpha=MaterialFunctionCall(Float2=@n10) {materialFunction=&BreakOutFloat2Components}.G
  )
  n13 = MaterialFunctionCall(
   UVs=@n10
   Texture Scale=MaterialFunctionCall(X=Multiply(A=@n12, B=@n11.Width), Y=Constant() {r=1}) {materialFunction=&MakeFloat2}.Result
  ) {materialFunction=&ScaleUVsByCenter}
  n14 = MaterialFunctionCall(Float2=@n13.UVs) {materialFunction=&BreakOutFloat2Components}
  n15 = Multiply(
   A=MaterialFunctionCall!(TexCoord=@n13.UVs, ClampChannel=Constant() {r=1}) {materialFunction=/Game/Effects/Fort_Effects/MaterialFunctions/MF_FX_TilingMethod}.Result
   B=Sine(None=Multiply(A=@n14.R) {constB=0.5})
  )
  n16 = Saturate(None=Multiply(A=Saturate(None=@n15)) {constB=20})
  n17 = ParticleColor()
  n18 = Multiply(A=Add(A=Multiply(A=@n9), B=Multiply(A=@n8, B=@n16)), B=@n17.RGB)
  n19 = VectorParameter() {parameterName=Main color, defaultValue=<rgba 1 0.3833 0.7682 0>}
  n20 = Power(Base=ComponentMask(None=TextureCoordinate()) {g=true})
  n21 = Add(A=TextureCoordinate() {uTiling=1.5, vTiling=2}, B=ParticleRandom())
  n22 = LinearInterpolate(Alpha=TextureSample(UVs=Panner(Coordinate=@n21) {speedY=0.3}) {texture=&T_Inky_Smoke_Tile}.R) {constB=1.4}
  n23 = Multiply(A=Saturate(None=Multiply(A=@n20) {constB=5}), B=Multiply(A=Power(Base=@n22) {constExponent=1}) {constB=3}) {constA=1}
  n24 = LinearInterpolate(
   A=@n18
   B=Add(A=Multiply(A=@n19.RGB, B=@n23), B=Subtract(A=@n18, B=@n23) {constB=0})
   Alpha=ScalarParameter() {defaultValue=1, parameterName=Add Dark color}
  )
  n25 = TextureCoordinate()
  n26 = LinearInterpolate(
   A=Constant() {r=0.2}
   B=DynamicParameter() {defaultValue=<rgba 1 1 0.4 1>, paramNames [Spline Ticken Intensity, Width, Width bottom, Appear]}.Width bottom
   Alpha=MaterialFunctionCall(Float2=@n25) {materialFunction=&BreakOutFloat2Components}.G
  )
  n27 = MaterialFunctionCall(
   UVs=@n25
   Texture Scale=MaterialFunctionCall(X=Multiply(A=@n26) {constB=5}, Y=Constant() {r=1}) {materialFunction=&MakeFloat2}.Result
  ) {materialFunction=&ScaleUVsByCenter}
  n28 = MaterialFunctionCall(TexCoord=@n27.UVs, ClampChannel=Constant() {r=1}) {materialFunction=&MF_FX_TilingMethod}
  n29 = Sine(None=Multiply(A=MaterialFunctionCall(Float2=@n27.UVs) {materialFunction=&BreakOutFloat2Components}.R) {constB=0.5})
  n30 = Saturate(None=Power(Base=Multiply(A=@n28.Result, B=@n29)) {constExponent=3})
  n31 = Multiply(A=@n30) {constB=20}
  n32 = Multiply(A=Multiply(A=Saturate(None=@n31), B=@n8), B=ParticleColor().RGB)
  n33 = ScalarParameter() {parameterName=Use simple glow}
  MaterialFunctionCall(Emissive=Multiply(
    A=MaterialFunctionCall(Color_Morning=@n1.RGB, Color_Day=@n2, Color_Evening=@n3.RGB, Color_Night=@n4.RGB) {materialFunction=&Color_Time_of_Day}.Result
    B=LinearInterpolate(A=@n24, B=@n32, Alpha=@n33)
   )) {materialFunction=&MF_FX_EmissiveAdjustments}
  n34 = ScalarParameter() {defaultValue=5, group=WPO, parameterName=Spline Thicken Width}
  Multiply(A=@n34) {constA=0.3, constB=0.5}
  n35 = ComponentMask(None=TextureCoordinate()) {g=true}
  n36 = Saturate(None=Multiply(A=Sine(None=@n35) {period=2}) {constB=5})
  n37 = ComponentMask(None=TextureCoordinate()) {g=true}
  n38 = Saturate(None=Multiply(A=Power(Base=@n37) {constExponent=1}) {constB=5})
  n39 = ParticleColor()
  n40 = TextureCoordinate()
  n41 = Multiply(A=Sine(None=ComponentMask(None=@n40) {g=true}) {period=2}) {constB=2}
  n42 = LinearInterpolate(
   A=Multiply(A=Multiply(A=@n17.A, B=@n16), B=Multiply(A=@n36, B=@n38))
   B=Multiply(A=Multiply(A=@n30, B=@n39.A), B=Saturate(None=@n41))
   Alpha=ScalarParameter() {parameterName=Use simple glow}
  )
  n43 = DepthFade(Opacity=@n42, FadeDistance=ScalarParameter() {defaultValue=30, parameterName=FadeDistance}) {fadeDistanceDefault=30}
  n44 = TextureCoordinate()
  n45 = Add(
   A=LinearInterpolate(Alpha=ComponentMask(None=@n44) {g=true}) {constA=-0.5}
   B=LinearInterpolate(Alpha=DynamicParameter() {defaultValue=<rgba 1 1 0.4 1>, paramNames [Spline Ticken Intensity, Width, Width bottom, Appear]}.Appear) {constA=-1, constB=2}
  )
  Reroute(None=Multiply(A=ShadingPathSwitch(Default=@n43, Deferred=@n43, Forward=@n42, Mobile=@n42), B=Saturate(None=@n45)))
  n46 = Constant()
  n47 = LinearInterpolate(
   B=DynamicParameter() {parameterIndex=2, paramNames [Random offset, none, none, none]}.Random offset
   Alpha=ScalarParameter() {parameterName=Use Random}
  )
  n48 = Multiply(
   A=Add(A=TextureCoordinate(), B=@n47) {constB=0}
   B=ComponentMask(None=VectorParameter() {group=WPO, parameterName=WPO_UVScale, defaultValue=<rgba 0 1.2 0 0>}.RGB) {g=true, r=true}
  )
  n49 = ComponentMask(None=VectorParameter() {group=WPO, parameterName=WPO_Speed, defaultValue=<rgba 0 0.3 0 0>}.RGB) {g=true, r=true}
  n50 = LinearInterpolate(Alpha=TextureSample(UVs=Panner(Coordinate=@n48, Speed=@n49) {speedY=0.5}) {samplerType=SAMPLERTYPE_LinearColor, texture=&T_NoisePack_LinearColor}.G)
  n51 = Add(A=TextureCoordinate(), B=@n47) {constB=0}
  n52 = ComponentMask(None=VectorParameter() {group=WPO, parameterName=WPO_UVScale002, defaultValue=<rgba 0 1 0 0>}.RGB) {g=true, r=true}
  n53 = TextureSample(UVs=Panner(Coordinate=Multiply(A=@n51, B=@n52), Speed=@n49) {speedY=0.5}) {samplerType=SAMPLERTYPE_Normal, texture=&T_Weap_Distort_N_02}
  n54 = TextureCoordinate()
  n55 = ScalarParameter() {defaultValue=0.3, parameterName=Blend offset}
  n56 = Multiply(
   A=Add(A=Add(A=@n54, B=@n55) {constB=0.2}) {constB=0}
   B=ComponentMask(None=VectorParameter() {group=WPO, parameterName=WPO_Blend_UVScale, defaultValue=<rgba 0 1 0 0>}.RGB) {g=true, r=true}
  )
  n57 = LinearInterpolate(
   A=ScalarParameter() {defaultValue=0.5, group=WPO, parameterName=WPO Blend Min}
   B=ScalarParameter() {defaultValue=4, group=WPO, parameterName=WPO Blend Max}
   Alpha=TextureSample(UVs=Panner(Coordinate=@n56, Speed=@n49) {speedY=0.3}) {samplerType=SAMPLERTYPE_LinearColor, texture=&T_NoisePack_LinearColor}.R
  ) {constA=0.5, constB=4}
  n58 = Add(A=Add(A=TextureCoordinate() {uTiling=0}, B=@n55) {constB=0.3}) {constB=0}
  n59 = ComponentMask(None=VectorParameter() {group=WPO, parameterName=WPO Blend mask Speed, defaultValue=<rgba 0 0.4 0 0>}.RGB) {g=true, r=true}
  n60 = LinearInterpolate(
   A=ScalarParameter() {group=WPO, parameterName=WPO Blend Min Mask}
   B=ScalarParameter() {defaultValue=1.2, group=WPO, parameterName=WPO Blend Max Mask}
   Alpha=TextureSample(UVs=Panner(Coordinate=@n58, Speed=@n59) {speedY=0.4}) {samplerType=SAMPLERTYPE_Normal, texture=&T_Weap_Distort_N_02}.R
  )
  n61 = Multiply(A=Power(Base=Multiply(A=@n60) {constB=2}, Exp=ScalarParameter() {defaultValue=1.5, group=WPO, parameterName=WPO Blend Power}) {constExponent=1.5}) {constB=0.5}
  n62 = Multiply(A=ComponentMask(None=TextureCoordinate()) {g=true}) {constB=7}
  n63 = Power(Base=ComponentMask(None=TextureCoordinate()) {g=true}) {constExponent=4}
  n64 = Multiply(
   A=Multiply(A=Multiply(A=@n57, B=@n61) {constB=0.5}, B=Saturate(None=@n62))
   B=LinearInterpolate(Alpha=Saturate(None=@n63)) {constA=1, constB=1.5}
  )
  n65 = ScalarParameter() {defaultValue=0.8, parameterName=Add blend}
  n66 = LinearInterpolate(Alpha=Add(A=LinearInterpolate(A=@n50, B=@n53.R) {constAlpha=0.6}, B=LinearInterpolate(B=@n64, Alpha=@n65)) {constB=0}) {constB=1.5}
  n67 = OneMinus(None=ComponentMask(None=TextureCoordinate()) {g=true})
  n68 = ScalarParameter() {defaultValue=5, group=WPO, parameterName=Bottom Mask}
  n69 = TextureCoordinate()
  n70 = Power(Base=Sine(None=ComponentMask(None=@n69) {g=true}) {period=2})
  n71 = LinearInterpolate(
   A=Saturate(None=Multiply(A=@n67, B=@n68) {constB=5})
   B=Saturate(None=Multiply(A=@n70) {constB=5})
   Alpha=ScalarParameter() {group=WPO, parameterName=Use both sides mask}
  )
  n72 = TextureCoordinate() {uTiling=0}
  n73 = ParticleRandom()
  n74 = Panner(Coordinate=Add(A=@n72, B=@n73)) {speedY=0.5}
  n75 = Panner(Coordinate=Add(A=@n72, B=@n73)) {speedY=0.5}
  n76 = LinearInterpolate(Alpha=LinearInterpolate(
    A=TextureSample(UVs=@n74) {samplerType=SAMPLERTYPE_Normal, texture=&T_Weap_Distort_N_02}.R
    B=TextureSample(UVs=@n75) {samplerType=SAMPLERTYPE_LinearColor, texture=&T_NoisePack_LinearColor}.R
   )) {constB=1.5}
  n77 = Add(
   A=Multiply(A=Multiply(A=@n66, B=@n71), B=Constant3Vector() {constant=<rgba 1 0 0 1>})
   B=Multiply(A=Multiply(A=@n76, B=@n71), B=VectorParameter() {parameterName=WPO Y, defaultValue=<rgba 0 0.5 0 1>}.RGB)
  ) {constB=0}
  n78 = Multiply(
   A=Add(A=Transform(None=@n77) {transformSourceType=TRANSFORMSOURCE_Local}) {constB=0}
   B=Multiply(A=ScalarParameter() {defaultValue=30, group=WPO, parameterName=WPO Intensity})
  ) {constB=30}
  n79 = Multiply(
   A=@n34
   B=DynamicParameter() {defaultValue=<rgba 1 1 0.4 1>, paramNames [Spline Ticken Intensity, Width, Width bottom, Appear]}.Spline Ticken Intensity
  )
  n80 = MaterialFunctionCall(WidthBase=@n79, WidthTip=@n79) {materialFunction=&SplineThicken}
  n81 = ScalarParameter() {group=WPO, parameterName=Use SplineTicken}
  Reroute(None=Add(A=Add(A=@n46, B=@n78), B=LinearInterpolate(B=@n80.WorldPositionOffset, Alpha=@n81)) {constB=0})
  n82 = DynamicParameter() {desc=Parameters set by Niagara, parameterIndex=1, paramNames [X,Y,Z,WPOIntensity]}
  n83 = MaterialFunctionCall(Float3=Subtract(
    A=MaterialFunctionCall(X=@n82.X, Y=@n82.X, Z=@n82.X) {materialFunction=&MakeFloat3}.Result
    B=ParticlePositionWS()
   )) {materialFunction=&BreakOutFloat3Components}
  n84 = Clamp(
   None=@n83.R
   Min=ScalarParameter() {defaultValue=-100, group=WPO, parameterName=Clamp_min}
   Max=ScalarParameter() {defaultValue=100, group=WPO, parameterName=Clamp_max}
  ) {maxDefault=1000, minDefault=-1000}
  n85 = Clamp(
   None=@n83.G
   Min=ScalarParameter() {defaultValue=-100, group=WPO, parameterName=Clamp_min}
   Max=ScalarParameter() {defaultValue=100, group=WPO, parameterName=Clamp_max}
  ) {maxDefault=1000, minDefault=-1000}
  n86 = Clamp(None=@n83.B, Min=ScalarParameter() {group=WPO, parameterName=Clamp_Z_min}, Max=ScalarParameter() {group=WPO, parameterName=Clamp_Z_max}) {maxDefault=1000, minDefault=-1000}
  n87 = TextureCoordinate()
  n88 = Power(Base=ComponentMask(None=OneMinus(None=@n87)) {g=true}, Exp=ScalarParameter() {defaultValue=3, group=WPO, parameterName=WPOBendPower})
  StaticSwitchParameter(True=LinearInterpolate(
    B=MaterialFunctionCall(X=@n84, Y=@n85, Z=@n86) {materialFunction=&MakeFloat3}.Result
    Alpha=Multiply(A=@n88, B=@n82.WPOIntensity)
   )) {group=WPO, parameterName=Use_Interpolate WPO}
 ]
}
