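# Glowing material for the Fuchsia female glider mesh, usable on skeletal meshes and Niagara mesh particles.
# Pre-skinned position, mirrored in X, or mesh UVs via "Use tail material", drive a curve-atlas ramp over panning color textures and smoke-distorted electric noise highlights.
# Fresnel and an optional gradient with fake sun/moon specular are added.
# Time-of-day tint and optional particle color scale the result before FX emissive adjustments.
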Material /BRCosmetics/Gadgets/Assets/VinderTech_GliderChute/Glider_Fuchsia_Female/FX/M_Glider_Fuchsia_Female_mesh
graph {
 properties {
  bUsedWithMeshDeformer=true
  bUsedWithNiagaraMeshParticles=true
  bUsedWithSkeletalMesh=true
  pixelDepthOffsetMode=PDOM_Legacy
  usedShadingModels=Default Lit
 }
 nodes [
  n1 = ParticleColor()
  LinearInterpolate(B=@n1.A, Alpha=ScalarParameter() {parameterName=Use particle color}) {constA=1}
  n2 = VectorParameter() {parameterName="Color_Morning (V3)", defaultValue=<rgba 0.4 0.4 0.4 0>}
  n3 = Constant() {r=1}
  n4 = VectorParameter() {parameterName="Color_Evening (V3)", defaultValue=<rgba 0.4 0.4 0.4 0>}
  n5 = VectorParameter() {parameterName="Color_Night (V3)", defaultValue=<rgba 0.7188 0.4735 0.6013 0>}
  n6 = Multiply(A=VectorParameter() {parameterName=Fresnel color, defaultValue=<rgba 5.7287 1.5416 20 0>}.RGB, B=Fresnel() {baseReflectFraction=0})
  n7 = TextureSample(UVs=TextureCoordinate()) {samplerType=SAMPLERTYPE_Grayscale, texture=&LowResBlurredNoise}
  n8 = MaterialFunctionCall(Float3=VertexInterpolator(VS=PreSkinnedPosition()).PS) {materialFunction=&BreakOutFloat3Components}
  n9 = Constant()
  n10 = Abs(None=Subtract(A=@n8.R, B=@n9))
  n11 = Subtract(A=@n8.B, B=@n9)
  n12 = LinearInterpolate(
   A=Multiply(A=AppendVector(A=@n10, B=@n11), B=Constant() {r=0.01})
   B=TextureCoordinate()
   Alpha=ScalarParameter() {parameterName=Use tail material}
  )
  n13 = LinearInterpolate(Alpha=MaterialFunctionCall(Float2=@n12) {materialFunction=&BreakOutFloat2Components}.R) {constA=-0.3}
  n14 = CurveAtlasRowParameter(CurveTime=Power(Base=Saturate(None=@n13)) {constExponent=1.5}) {
   atlas=/Game/Athena/LinearColorCurves/A_Atlas_Turtleneck_Glider
   curve=/Game/Gadgets/Assets/VinderTech_GliderChute/Glider_Fuchsia_Female/FX/Glider_Fuchsia_Female_Mesh_Curve
   defaultValue=19
   sortPriority=0
  }
  n15 = VectorParameter() {parameterName=Base Tex Color, defaultValue=<rgba 1 1 1 1>}
  n16 = Panner(Coordinate=TextureCoordinate(), Speed=ComponentMask(None=VectorParameter() {parameterName=Panner speed}.RGB) {g=true, r=true})
  n17 = Add(A=@n16) {constB=0}
  n18 = Constant3Vector() {constant=<rgba 0.6418 0.3832 1 0>}
  n19 = Constant3Vector() {constant=<rgba 0.1704 0.0112 1 0>}
  n20 = TextureSample(UVs=@n16) {texture=&T_Inky_Smoke_Tile}
  n21 = Desaturation(
   None=LinearInterpolate(
    A=TextureSampleParameter2D(UVs=@n17) {parameterName=Color_Tex, texture=/Game/Gadgets/Assets/VinderTech_GliderChute/Glider_Cube_Queen_Female/FX/T_Cheap_blue_space_color}.RGB
    B=LinearInterpolate(A=@n18, B=@n19, Alpha=@n20.R)
   ) {constAlpha=0.03}
   Fraction=ScalarParameter() {parameterName=Base Color Tex Des}
  )
  n22 = Multiply(
   A=Multiply(A=Multiply(A=@n14, B=@n7.R)) {constB=1.5}
   B=Multiply(A=Multiply(A=@n15.RGB, B=@n21), B=ScalarParameter() {defaultValue=30, parameterName=Tex_Intensity}) {constB=30}
  )
  n23 = TextureSampleParameter2D() {parameterName=HighLight_Tex, texture=&T_Cheap_blue_space_color}
  n24 = Multiply(A=@n23.RGB, B=ScalarParameter() {defaultValue=40, parameterName=Noise intensity}) {constB=40}
  n25 = TextureSample(UVs=Panner(Coordinate=@n12) {speedX=0.05, speedY=-0.05}) {texture=&T_Inky_Smoke_Tile}
  n26 = Multiply(
   A=Add(A=@n12, B=Multiply(A=@n25.R) {constB=0.2})
   B=ComponentMask(None=VectorParameter() {parameterName=Noise UV Scale, defaultValue=<rgba 0.5 1 0 1>}.RGB) {g=true, r=true}
  )
  n27 = ComponentMask(None=VectorParameter() {parameterName=Noise speed, defaultValue=<rgba -0.15 -0.1 0 0>}.RGB) {g=true, r=true}
  n28 = Power(Base=TextureSample(UVs=Panner(Coordinate=@n26, Speed=@n27) {speedX=-0.15, speedY=-0.1}) {samplerType=SAMPLERTYPE_LinearColor, texture=/Game/Characters/Player/Male/Medium/Bodies/M_MED_SRDrift/Textures/T_ElectricNoise}.R)
  n29 = TextureCoordinate()
  n30 = Multiply(
   A=VectorParameter() {parameterName=Gradient color, defaultValue=<rgba 0.2814 0.1527 1 0>}.RGB
   B=Power(Base=ComponentMask(None=@n29) {g=true})
  )
  n31 = CollectionParameter() {
   collection=&FortniteMaterialParameters
   parameterName=SunAndMoonModelDirectionalVector
  }
  n32 = DotProduct(A=ComponentMask(None=@n31) {b=true, g=true, r=true}, B=ReflectionVectorWS())
  n33 = Multiply(A=Power(Base=Clamp(None=@n32), Exp=Constant() {r=20}) {constExponent=20}, B=Constant() {r=2})
  n34 = Add(
   A=Add(A=Add(A=@n6, B=@n22), B=Multiply(A=@n24, B=@n28) {constA=1})
   B=LinearInterpolate(B=Add(A=@n30, B=@n33) {constB=0}, Alpha=ScalarParameter() {defaultValue=1, parameterName=Use Fake spec}) {constB=0}
  )
  n35 = LinearInterpolate(B=@n1.RGB, Alpha=ScalarParameter() {parameterName=Use particle color}) {constA=1}
  MaterialFunctionCall(Emissive=Multiply(
    A=MaterialFunctionCall(Color_Morning=@n2.RGB, Color_Day=@n3, Color_Evening=@n4.RGB, Color_Night=@n5.RGB) {materialFunction=&Color_Time_of_Day}.Result
    B=Multiply(A=@n34, B=@n35)
   )) {materialFunction=&MF_FX_EmissiveAdjustments}
 ]
}
